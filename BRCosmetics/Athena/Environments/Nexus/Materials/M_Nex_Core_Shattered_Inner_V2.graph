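# Default Lit material for the inner shattered Nexus core, usable on Niagara mesh particles, with a glowing surface and animated vertex offsets.
# Emissive layers three tinted UV-gradient colors, with brightness lerped by Progress from a parameter or the NexCollection MPC.
# MF_Nex_Core_WPO supplies shatter spread, optionally driven by MPC time, pulse, shake and progress.
# Switches add RiftBeacon offset and thickness adjustment along vertex-color-encoded directions, weighted by inverse alpha.
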Material /BRCosmetics/Athena/Environments/Nexus/Materials/M_Nex_Core_Shattered_Inner_V2
graph {
 properties {bUsedWithNiagaraMeshParticles=true, pixelDepthOffsetMode=PDOM_Legacy, usedShadingModels=Default Lit}
 nodes [
  Constant()
  Constant() {r=1}
  n1 = ScalarParameter() {defaultValue=10, parameterName=Brightness Multiplier}
  n2 = ScalarParameter() {defaultValue=0.5, parameterName=Low Brightness Multiplier}
  n3 = Power(Base=StaticSwitchParameter(
    True=CollectionParameter() {collection=/Game/Athena/Environments/Nexus/Materials/NexCollection, parameterName=Progress}
    False=ScalarParameter() {parameterName=Progress}
   ) {parameterName=UseMPC}) {constExponent=0.6}
  n4 = ComponentMask(None=TextureCoordinate()) {g=true}
  n5 = Saturate(None=LinearInterpolate(B=ScalarParameter() {defaultValue=-9, parameterName=EdgeGradientSubtract}, Alpha=@n4) {constA=1, constB=-10})
  n6 = ScalarParameter() {defaultValue=1, parameterName=Core Outer Gradient Power}
  n7 = Add(
   A=Multiply(A=VectorParameter() {parameterName=Top, defaultValue=<rgba 0.2078 0.191 0.4844 1>}.RGB, B=Power(Base=@n5) {constExponent=4})
   B=Multiply(A=VectorParameter() {parameterName=Outer, defaultValue=<rgba 0.1586 0.6943 1 1>}.RGB, B=Power(Base=@n4, Exp=@n6))
  )
  n8 = TextureCoordinate()
  n9 = Multiply(
   A=VectorParameter() {parameterName=Inner, defaultValue=<rgba 13.3709 31.7795 50 1>}.RGB
   B=Power(Base=ComponentMask(None=@n8) {g=true}, Exp=ScalarParameter() {defaultValue=10, parameterName=Fresnel2Power})
  )
  MaterialFunctionCall(Emissive=Multiply(A=LinearInterpolate(A=@n1, B=@n2, Alpha=@n3) {constA=20, constB=0.5}, B=Add(A=@n7, B=@n9))) {materialFunction=&MF_FX_EmissiveAdjustments}
  MaterialProxyReplace(Realtime=Constant3Vector(), MaterialProxy=Constant3Vector() {constant=<rgba 0.1529 0.451 0.5922 1>})
  ScalarParameter() {parameterName=Roughness}
  n10 = StaticSwitchParameter(
   True=MaterialFunctionCall!(
    Time Multiplier=CollectionParameter() {collection=&NexCollection, parameterName=TimeMultiplier}
    Lerp Shatter Pulse=CollectionParameter() {collection=&NexCollection, parameterName=PulseLerp}
    Shake=CollectionParameter() {collection=&NexCollection, parameterName=Shake}
    Progress=CollectionParameter() {collection=&NexCollection, parameterName=Progress}
   ) {materialFunction=/Game/Athena/Environments/Nexus/Materials/MF_Nex_Core_WPO}.WPOSpread
   False=MaterialFunctionCall() {materialFunction=&MF_Nex_Core_WPO}.WPOSpread
  ) {parameterName=UseMPC}
  n11 = StaticSwitchParameter(True=Add(A=@n10, B=MaterialFunctionCall() {materialFunction=/Game/Environments/Props/RiftBeacon/FX/RiftBeacon_NexusWPO}.Result), False=@n10) {parameterName=Used with RiftBeacon}
  n12 = VertexColor()
  n13 = Subtract(A=@n12) {constB=0.5}
  n14 = Add(
   A=Transform(None=Multiply(A=@n13) {constB=2}) {transformSourceType=TRANSFORMSOURCE_Local}
   B=Transform(None=Constant3Vector()) {transformSourceType=TRANSFORMSOURCE_Local}
  )
  n15 = Multiply(A=ScalarParameter() {defaultValue=-50, parameterName=Thickness}, B=OneMinus(None=@n12.A))
  StaticSwitchParameter(True=Add(A=@n11, B=Multiply(A=@n14, B=@n15)), False=@n11) {parameterName=Adjust Thickness}
 ]
}
